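# Post-copy AnimBP for the Moisty Merman male medium body cosmetic, adding secondary motion to fins, zipper and shoulders.
# On initialization it caches the owner's skeletal mesh component, whose pose the AnimGraph copies.
# A weighted constraint ties the zipper to spine, neck and clavicle bones, then chained root-relative AnimDynamics nodes simulate arm and calf fins with angular springs, the zipper, and low-alpha shoulders.
# Zipper and shoulder simulations have LOD thresholds; the update event is empty.
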AnimBlueprint /BRCosmetics/Characters/Player/Male/Medium/Bodies/M_MED_Moisty_Merman/Meshes/M_MED_Moisty_Merman_AnimBP
target_skeleton=&/M_MED_Moisty_Merman_Skeleton
graphs [
 AnimGraph {nodes [
   n1 = Get MeshToCopy() -> (MeshToCopy:Skeletal Mesh Component Object Reference)
   n2 = "Constraint - dyn_zipper (PPPPP)"!(
    ComponentPose=Local To Component(LocalPose=Copy Pose From Mesh(SourceMeshComponent=@n1))
    Alpha=1
    ConstraintWeights_0=0.2132
    ConstraintWeights_1:float=0.0518
    ConstraintWeights_2:float=0.2029
    ConstraintWeights_3:float=0.3649
    ConstraintWeights_4:float=0.1673
   ) -> (Pose) {
    alphaBoolBlend {blendOption=Linear}
    boneToModify {boneName=dyn_zipper}
    constraintSetup [
     {targetBone {boneName=clavicle_r}}
     {targetBone {boneName=neck_01}}
     {targetBone {boneName=spine_03}}
     {targetBone {boneName=spine_05}}
     {targetBone {boneName=clavicle_l}}
    ]
    constraintWeights [1,1,1,1,1]
   }
   n3 = Anim Dynamics - Bone: dyn_armFin_l(ComponentPose=@n2, Alpha=1, ExternalForce="0,0,0") {
    alpha=0.9
    angularDampingOverride=0.9999
    angularSpringConstant=1500
    bAngularSpring=true
    bGravityOverrideInSimSpace=true
    bOverrideAngularDamping=true
    numSolverIterationsPostUpdate=4
    numSolverIterationsPreUpdate=16
    simulationSpace=RootRelative
    alphaBoolBlend {blendOption=Linear}
    boundBone {boneName=dyn_armFin_l}
    physicsBodyDefinitions [{
      sphereCollisionRadius=0
      boundBone {boneName=dyn_armFin_l}
      boxExtents=<15 5 5>
      constraintSetup {angularTargetAxis=AxisZ, twistAxis=AxisY, angularLimitsMax=<30 0 0>, angularLimitsMin=<-30 0 0>, angularTarget=<0 0 1>}
      localJointOffset=<0 0 0>
     }]
   }
   n4 = Anim Dynamics - Bone: dyn_armFin_r(ComponentPose=@n3, Alpha=1, ExternalForce="0,0,0") {
    alpha=0.9
    angularDampingOverride=0.9999
    angularSpringConstant=1500
    bAngularSpring=true
    bGravityOverrideInSimSpace=true
    bOverrideAngularDamping=true
    numSolverIterationsPostUpdate=4
    numSolverIterationsPreUpdate=16
    simulationSpace=RootRelative
    alphaBoolBlend {blendOption=Linear}
    boundBone {boneName=dyn_armFin_r}
    physicsBodyDefinitions [{
      sphereCollisionRadius=0
      boundBone {boneName=dyn_armFin_r}
      boxExtents=<15 5 5>
      constraintSetup {angularTargetAxis=AxisZ, twistAxis=AxisY, angularLimitsMax=<30 0 0>, angularLimitsMin=<-30 0 0>, angularTarget=<0 0 1>}
      localJointOffset=<0 0 0>
     }]
   }
   n5 = Anim Dynamics - Bone: dyn_calfFin_l(ComponentPose=@n4, Alpha=1, ExternalForce="0,0,0") {
    alpha=0.9
    angularDampingOverride=0.9999
    angularSpringConstant=1500
    bAngularSpring=true
    bGravityOverrideInSimSpace=true
    bOverrideAngularDamping=true
    numSolverIterationsPostUpdate=4
    numSolverIterationsPreUpdate=16
    simulationSpace=RootRelative
    alphaBoolBlend {blendOption=Linear}
    boundBone {boneName=dyn_calfFin_l}
    physicsBodyDefinitions [{
      sphereCollisionRadius=0
      boundBone {boneName=dyn_calfFin_l}
      boxExtents=<15 5 5>
      constraintSetup {angularTargetAxis=AxisZ, twistAxis=AxisY, angularLimitsMax=<30 0 0>, angularLimitsMin=<-30 0 0>, angularTarget=<0 0 1>}
      localJointOffset=<0 0 0>
     }]
   }
   n6 = Anim Dynamics - Bone: dyn_calfFin_r(ComponentPose=@n5, Alpha=1, ExternalForce="0,0,0") {
    alpha=0.9
    angularDampingOverride=0.9999
    angularSpringConstant=1500
    bAngularSpring=true
    bGravityOverrideInSimSpace=true
    bOverrideAngularDamping=true
    numSolverIterationsPostUpdate=4
    numSolverIterationsPreUpdate=16
    simulationSpace=RootRelative
    alphaBoolBlend {blendOption=Linear}
    boundBone {boneName=dyn_calfFin_r}
    physicsBodyDefinitions [{
      sphereCollisionRadius=0
      boundBone {boneName=dyn_calfFin_r}
      boxExtents=<15 5 5>
      constraintSetup {angularTargetAxis=AxisZ, twistAxis=AxisY, angularLimitsMax=<30 0 0>, angularLimitsMin=<-30 0 0>, angularTarget=<0 0 1>}
      localJointOffset=<0 0 0>
     }]
   }
   n7 = Anim Dynamics - Bone: dyn_zipper(ComponentPose=@n6, Alpha=1) {
    angularDampingOverride=0.9999
    bGravityOverrideInSimSpace=true
    bOverrideAngularDamping=true
    lODThreshold=1
    numSolverIterationsPostUpdate=4
    numSolverIterationsPreUpdate=16
    simulationSpace=RootRelative
    alphaBoolBlend {blendOption=Linear}
    boundBone {boneName=dyn_zipper}
    physicsBodyDefinitions [{
      sphereCollisionRadius=0
      boundBone {boneName=dyn_zipper}
      boxExtents=<5 5 5>
      constraintSetup {angularTargetAxis=AxisZ, twistAxis=AxisZ, angularLimitsMax=<5 30 30>, angularLimitsMin=<-90 -30 -30>, angularTarget=<0 0 0>}
      localJointOffset=<0 0 -0.5>
     }]
   }
   n8 = Anim Dynamics - Bone: dyn_shoulder_l(ComponentPose=@n7, Alpha=0.2, ExternalForce="0,0,0") {
    bGravityOverrideInSimSpace=true
    lODThreshold=1
    numSolverIterationsPostUpdate=4
    numSolverIterationsPreUpdate=16
    simulationSpace=RootRelative
    alphaBoolBlend {blendOption=Linear}
    boundBone {boneName=dyn_shoulder_l}
    physicsBodyDefinitions [{
      sphereCollisionRadius=0
      boundBone {boneName=dyn_shoulder_l}
      boxExtents=<10 20 10>
      constraintSetup {angularLimitsMax=<15 15 0>, angularLimitsMin=<-15 -15 0>, angularTarget=<0 0 0>, linearAxesMin=<-5 0 0>}
      localJointOffset=<10 0 -5>
     }]
   }
   n9 = Anim Dynamics - Bone: dyn_shoulder_r(ComponentPose=@n8, Alpha=0.2, ExternalForce="0,0,0") {
    bGravityOverrideInSimSpace=true
    lODThreshold=1
    numSolverIterationsPostUpdate=4
    numSolverIterationsPreUpdate=16
    simulationSpace=RootRelative
    alphaBoolBlend {blendOption=Linear}
    boundBone {boneName=dyn_shoulder_r}
    physicsBodyDefinitions [{
      sphereCollisionRadius=0
      boundBone {boneName=dyn_shoulder_r}
      boxExtents=<10 20 10>
      constraintSetup {angularLimitsMax=<15 15 0>, angularLimitsMin=<-15 -15 0>, angularTarget=<0 0 0>, linearAxesMin=<-5 0 0>}
      localJointOffset=<-10 0 3.5>
     }]
   }
   Output Pose(Result=Component To Local(ComponentPose=@n9)) {layerGroup=None}
  ]}
 EventGraph {nodes [
   Event BlueprintInitializeAnimation()
   > Set MeshToCopy(MeshToCopy:Skeletal Mesh Component Object Reference=GetOwnerSkeletalMeshComponent()) -> (Output_Get:Skeletal Mesh Component Object Reference)
   TryGetPawnOwner()
   Event BlueprintUpdateAnimation()
  ]}
]
variables [MeshToCopy {type=SkeletalMeshComponent}]
